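# Unlit translucent environment material for Apollo fireflies, rendering mesh quads as glowing sprites that default to appearing only at night.
# Vertex colors and world position vary pulse and hue of the time-of-day color.
# World position offset uses the Sprite function, sine wander and looping directional travel, with UV1 offsetting phases and size pulsing along the loop.
# Opacity combines flare texture, sphere mask, time-of-day value and distance LOD fade, with optional depth and loop-edge fades.
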Material /Game/Effects/Fort_Effects/Materials/Environment/Apollo/M_Apollo_FireFlies
graph {
 properties {
  bUsedWithInstancedStaticMeshes=true
  blendMode=BLEND_Translucent
  pixelDepthOffsetMode=PDOM_Legacy
  shadingModel=MSM_Unlit
  usedShadingModels=Unlit
 }
 nodes [
  n1 = VertexColor()
  n2 = MaterialFunctionCall(Speed=@n1.R) {materialFunction=&DebugTimeSine}
  n3 = Divide(A=WorldPosition().XYZ, B=ScalarParameter() {defaultValue=128, parameterName=RandomPulseDivisor}) {constB=128}
  n4 = Add(
   A=MaterialFunctionCall(Speed=@n1.G) {materialFunction=&DebugTimeSine}.Result
   B=ComponentMask(None=Frac(None=@n3)) {g=true}
  )
  n5 = LinearInterpolate(
   A=Constant()
   B=Constant() {r=1}
   Alpha=MaterialFunctionCall(Speed=@n1.B) {materialFunction=&DebugTimeSine}.Result
  ) {constA=0.15, constB=0.85}
  n6 = VectorParameter() {parameterName="Color_Morning (V3)"}
  n7 = VectorParameter() {parameterName="Color_Day (V3)", defaultValue=<rgba 1 1 1 0>}
  n8 = VectorParameter() {parameterName="Color_Evening (V3)", defaultValue=<rgba 1 1 1 0>}
  n9 = VectorParameter() {parameterName="Color_Night (V3)", defaultValue=<rgba 1 1 1 0>}
  Multiply(
   A=Power(Base=LinearInterpolate(A=@n2.Result, B=@n4, Alpha=@n5), Exp=ScalarParameter() {defaultValue=1, parameterName=GlowContrast})
   B=MaterialFunctionCall(
    Hue Shift Percentage=Multiply(A=@n1.R) {constB=-0.06}
    Texture=MaterialFunctionCall(Color_Morning=@n6.RGB, Color_Day=@n7.RGB, Color_Evening=@n8.RGB, Color_Night=@n9.RGB) {materialFunction=&Color_Time_of_Day}.Result
   ) {materialFunction=&HueShift}.Result
  )
  n10 = Multiply(A=ScalarParameter() {defaultValue=0.25, group=WPO, parameterName=MinSizeMult}, B=VertexColor().G)
  n11 = ScalarParameter() {defaultValue=500, group=WPO, parameterName=maxSize}
  n12 = Time() {bOverride_Period=true, period=3}
  n13 = Divide(A=Time() {bOverride_Period=true, period=6}) {constB=6}
  n14 = Divide(A=Time() {bOverride_Period=true, period=4}) {constB=4}
  n15 = ComponentMask(None=TextureCoordinate() {coordinateIndex=1}) {g=true}
  n16 = Add(
   A=StaticSwitchParameter(True=Divide(A=@n12) {constB=3}, False=StaticSwitchParameter(True=@n13, False=@n14) {parameterName=UseVerySlowSprites}) {defaultValue=true, parameterName=UseFastSprites}
   B=Multiply(A=Subtract(A=@n15) {constB=0.5}) {constB=-2}
  )
  n17 = Add(A=@n16) {constB=-0.25}
  n18 = LinearInterpolate(Alpha=ConstantBiasScale(None=Sine(None=@n17)))
  n19 = Multiply(
   A=LinearInterpolate(A=Multiply(A=@n10, B=@n11), B=@n11, Alpha=Power(Base=@n18) {constExponent=0.6})
   B=MaterialFunctionCall(Distance=ScalarParameter() {defaultValue=500, group=WPO, parameterName=NearCamShrinkDist}) {materialFunction=&Near_Camera_Fade}.Result
  )
  n20 = ScalarParameter() {defaultValue=1, parameterName=YScale}
  n21 = StaticSwitchParameter(True=AppendVector(A=@n19, B=Multiply(A=@n20, B=@n19)), False=@n19) {parameterName=UseXYScale?}
  n22 = NamedRerouteUsage()
  n23 = MaterialFunctionCall!(
   Center Location=WorldPosition().XYZ
   0-1 UVs
   XY scale=StaticSwitchParameter(True=Multiply(A=@n21, B=@n22), False=@n21) {parameterName=ScaleBasedOnDayNightValues}
   OverridePivot
   "Normals (Optional)"
  ) {materialFunction=&Sprite}
  n24 = Multiply(A=Time(), B=ScalarParameter() {defaultValue=0.001, group=WPO, parameterName=VarianceSpeed}) {constB=0.05}
  n25 = ComponentMask(None=TextureCoordinate() {coordinateIndex=1}) {r=true}
  n26 = Add(
   A=Multiply(A=Time(), B=ScalarParameter() {defaultValue=0.4, group=WPO, parameterName=DriftSpeed})
   B=Sine(None=Add(A=@n24, B=@n25)) {period=0.17}
  )
  n27 = AppendVector(A=Sine(None=@n26) {period=0.56}, B=Sine(None=@n26) {period=0.7})
  n28 = Sine(None=@n26) {period=0.82}
  n29 = ScalarParameter() {defaultValue=1, group=WPO, parameterName=MinHorizontalDrift}
  n30 = ScalarParameter() {defaultValue=1, group=WPO, parameterName=MaxHorizontalDrift}
  n31 = Sine(None=Add(A=@n24, B=@n25)) {period=0.14}
  n32 = Multiply(A=ObjectRadius(), B=Constant() {r=0.5}) {constB=0.5}
  n33 = LinearInterpolate(A=Multiply(A=@n32) {constB=-1}, B=@n32, Alpha=Frac(None=@n16))
  n34 = ScalarParameter() {defaultValue=1, group=WPO, parameterName=OverallSpeed}
  n35 = VectorParameter() {group=WPO, parameterName=OverallDirection, defaultValue=<rgba 1 0 0 1>}
  n36 = VectorParameter() {group=WPO, parameterName=OverallDirectionLow, defaultValue=<rgba 1 0 0 1>}
  n37 = VertexColor()
  n38 = Add(
   A=Multiply(A=AppendVector(A=@n27, B=@n28), B=LinearInterpolate(A=@n29, B=@n30, Alpha=@n31))
   B=Multiply(A=Multiply(A=@n33, B=@n34), B=LinearInterpolate(A=@n35.RGB, B=@n36.RGB, Alpha=@n37.G))
  )
  n39 = Constant()
  n40 = Multiply(A=VertexNormalWS(), B=ScalarParameter() {defaultValue=10, group=WPO, parameterName=SizeScale})
  n41 = WorldPosition()
  n42 = CameraPositionWS()
  n43 = Clamp(None=Divide(A=Distance(A=@n41.XYZ, B=@n42), B=ScalarParameter() {defaultValue=10, group=WPO, parameterName=SizeScale_MaxDistance}) {constB=15000})
  Multiply(B=Add(A=Add(A=@n23.World position offset, B=@n38), B=LinearInterpolate(A=@n39, B=@n40, Alpha=@n43)))
  n44 = MaterialFunctionCall(
   Value_Morning=ScalarParameter() {parameterName="Value_Morning (S)"}
   Value_Day=ScalarParameter() {parameterName="Value_Day (S)"}
   Value_Evening=ScalarParameter() {parameterName="Value_Evening (S)"}
   Value_Night=ScalarParameter() {defaultValue=1, parameterName="Value_Night (S)"}
  ) {materialFunction=&Scalar_Time_of_Day}
  NamedRerouteDeclaration(None=Clamp(None=@n44.Result)) {name=DayToNightValues, nodeColor=<rgba 0.6875 0 1 1>}
  n45 = ScalarParameter() {defaultValue=15000, group=Scaler, parameterName="NearFadeLength (S)"}
  n46 = ScalarParameter() {defaultValue=10000, group=Scaler, parameterName="NearFadeOffset (S)"}
  n47 = ScalarParameter() {defaultValue=15000, group=Scaler, parameterName="FarFadeLength (S)"}
  n48 = ScalarParameter() {defaultValue=75000, group=Scaler, parameterName="FarFadeOffset (S)"}
  n49 = ScalarParameter() {group=Scaler, parameterName="Opacity Near Distance Bias Opacity (S)"}
  n50 = ScalarParameter() {group=Scaler, parameterName="Opacity Far Distance Bias Opacity (S)"}
  n51 = TextureSampleParameter2D() {parameterName=Texture, texture=/Game/Effects/Fort_Effects/Textures/Flares/T_Flare_Spikeball}
  n52 = Multiply(
   A=SphereMask(
    A=TextureCoordinate()
    B=Constant2Vector() {g=0.5, r=0.5}
    Radius=ScalarParameter() {defaultValue=0.5, parameterName=Sprites_Radius}
    Hardness=ScalarParameter() {parameterName=Sprite_Hardness}
   )
   B=ScalarParameter() {defaultValue=1, parameterName=Opacity}
  )
  n53 = Multiply(
   A=Power(
    Base=MaterialFunctionCall!(
     NearFadeLength=@n45
     NearFadeOffset=@n46
     NearFadeCurve
     NearFadeStartingScale
     NearFadeEndingScale
     FarFadeLength=@n47
     FarFadeOffset=@n48
     FarFadeCurve
     FarFadeEndingScale
     Opacity Near Distance Bias=@n49
     Opacity Far Distance Bias=@n50
    ) {materialFunction=/Game/Effects/Fort_Effects/Materials/_Functions/MF_SmoothLod}.Opacity
    Exp=ScalarParameter() {defaultValue=4, group=Opacity, parameterName=OpactiyDistScalePower}
   )
   B=Multiply(A=Multiply(A=@n51.G, B=@n52), B=@n44.Result)
  )
  n54 = ScalarParameter() {parameterName=DepthFadeDistance}
  n55 = StaticSwitchParameter(True=Multiply(A=@n53, B=DepthFade(FadeDistance=@n54)), False=@n53) {parameterName=UseDepthFade?}
  n56 = Add(A=@n16) {constB=-0.25}
  n57 = LinearInterpolate(Alpha=ConstantBiasScale(None=Sine(None=@n56)))
  StaticSwitchParameter(True=Multiply(A=@n55, B=Power(Base=@n57) {constExponent=0.6}), False=@n55) {parameterName=UseFadeOpacityAtEdge?}
 ]
}
