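# Unlit translucent material for the Apollo power plant smoke pillar, with animated billowing and a fiery gradient glow.
# Time-of-day colors use sun-driven normal shading and a Schlick phase function for scattering.
# Two panning cloud samples drive vertex displacement and, interpolated, pixel detail and glow, plus a wave offset replaced on low/mobile.
# Opacity combines UV ramp, view-angle, near-camera and depth fades, cut by a biased custom-depth comparison.
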Material /Game/Effects/Fort_Effects/Materials/Environment/Apollo/M_Apollo_Powerplant_SmokePillar_Transparent
graph {
 properties {
  allowTranslucentCustomDepthWrites=true
  bUsedWithInstancedStaticMeshes=true
  blendMode=BLEND_Translucent
  pixelDepthOffsetMode=PDOM_Legacy
  shadingModel=MSM_Unlit
  usedShadingModels=Unlit
 }
 nodes [
  n1 = VectorParameter() {parameterName=ShadowColorDay, defaultValue=<rgba 0.1679 0.1885 0.2049 1>}
  n2 = VectorParameter() {parameterName=ShadowColorMorning, defaultValue=<rgba 0.1552 0.1846 0.191 1>}
  n3 = VectorParameter() {parameterName=ShadowColorEvening, defaultValue=<rgba 0.1181 0.0709 0.0828 1>}
  n4 = VectorParameter() {parameterName=ShadowColorNight, defaultValue=<rgba 0.0952 0.1258 0.1319 1>}
  n5 = MaterialFunctionCall(
   Color_Morning=VectorParameter() {parameterName=LightColorMorning, defaultValue=<rgba 0.8992 0.5725 0.9062 1>}.RGB
   Color_Day=VectorParameter() {parameterName=LightColorDay, defaultValue=<rgba 0.8177 0.945 1 1>}.RGB
   Color_Evening=VectorParameter() {parameterName=LightColorEvening, defaultValue=<rgba 0.7969 0.4746 0.303 1>}.RGB
   Color_Night=VectorParameter() {parameterName=LightColorNight, defaultValue=<rgba 0.2019 0.2816 0.316 1>}.RGB
  ) {materialFunction=&Color_Time_of_Day}
  n6 = CollectionParameter() {
   collection=&FortniteMaterialParameters
   parameterName=SunlightDirectionalVector
  }
  n7 = Normalize(VectorInput=ComponentMask(None=@n6) {b=true, g=true, r=true})
  n8 = TextureCoordinate()
  n9 = AppendVector(
   A=Multiply(A=ScalarParameter() {defaultValue=2, parameterName=Tile_N_X}, B=ComponentMask(None=@n8) {r=true})
   B=Multiply(A=ComponentMask(None=@n8) {g=true}, B=ScalarParameter() {defaultValue=6, parameterName=Tile_N_Y})
  )
  n10 = ComponentMask(None=VectorParameter() {parameterName=FakeLight_Speed, defaultValue=<rgba 0.008 0.0125 0 0>}.RGB) {g=true, r=true}
  n11 = TextureCoordinate()
  n12 = MaterialFunctionCall(
   Normal=TextureSample(UVs=Panner(Coordinate=@n9, Speed=@n10)) {samplerType=SAMPLERTYPE_Normal, texture=/Game/Global/Textures/Default/Tileable_Patterns/2ndInfectedMapNormal}.RGB
   Flatness=LinearInterpolate(
    A=ScalarParameter() {defaultValue=0.5, parameterName=FlattenDetailNormal_Top}
    B=ScalarParameter() {defaultValue=0.5, parameterName=FlattenDetailNormal}
    Alpha=ComponentMask(None=@n11) {g=true}
   )
  ) {materialFunction=&FlattenNormal}
  n13 = LinearInterpolate(A=@n12.Result, B=VertexNormalWS())
  n14 = ScalarParameter() {defaultValue=0.1, parameterName=LerpVertNormal}
  n15 = ConstantBiasScale(None=DotProduct(
    A=Transform(None=@n7) {transformSourceType=TRANSFORMSOURCE_World, transformType=TRANSFORM_Tangent}
    B=LinearInterpolate(A=@n12.Result, B=@n13, Alpha=@n14)
   )) {bias=0.5, scale=0.6}
  n16 = ScalarParameter() {defaultValue=2.4, parameterName=SoftenShadow}
  n17 = LinearInterpolate(
   A=Multiply(
    A=MaterialFunctionCall(Color_Morning=@n1.RGB, Color_Day=@n2.RGB, Color_Evening=@n3.RGB, Color_Night=@n4.RGB) {materialFunction=&Color_Time_of_Day}.Result
    B=@n5.Result
   )
   B=@n5.Result
   Alpha=Clamp(None=Power(Base=@n15, Exp=@n16))
  )
  n18 = CollectionParameter() {
   collection=&FortniteMaterialParameters
   parameterName=SunlightDirectionalVector
  }
  MaterialExpressionDotProduct_0 = DotProduct(A=ComponentMask(None=@n18) {b=true, g=true, r=true}, B=CameraVectorWS())
  MaterialExpressionScalarParameter_7 = ScalarParameter() {defaultValue=0.4, parameterName=Anisotropy}
  n19 = Custom!(CosTheta=@MaterialExpressionDotProduct_0, k=@MaterialExpressionScalarParameter_7) {
   code="\tfloat Inner = (1 + k * CosTheta);\r\n\treturn (1 - k * k) / (4 * PI * Inner * Inner);\r\n\r\n"
   description=Schlick Phase
   inputs [
    {
     inputName=CosTheta
     input {expression=&self:MaterialExpressionDotProduct_0}
    }
    {
     inputName=k
     input {expression=&self:MaterialExpressionScalarParameter_7}
    }
   ]
  }
  n20 = Multiply(A=@n19, B=ScalarParameter() {defaultValue=28, parameterName=ScaleCloudColor})
  n21 = MaterialFunctionCall(Float2=Add(A=TextureCoordinate() {vTiling=2}, B=ScalarParameter() {group=WPO, parameterName=SmokeOffset})) {materialFunction=&BreakOutFloat2Components}
  n22 = TextureCoordinate()
  n23 = Power(
   Base=SphereMask(A=ComponentMask(None=@n22) {r=true}, B=Constant() {r=0.5}, Radius=Constant() {r=0.5}, Hardness=Constant() {r=0.5})
   Exp=ScalarParameter() {defaultValue=1, group=WPO, parameterName=Seam_Power}
  ) {constExponent=4}
  n24 = LinearInterpolate(
   A=ScalarParameter() {defaultValue=1, group=WPO, parameterName=ScaleU_Seam}
   B=ScalarParameter() {defaultValue=1, group=WPO, parameterName=ScaleU}
   Alpha=@n23
  )
  n25 = Multiply(
   A=OneMinus(None=MaterialFunctionCall() {materialFunction=&LinearGradient}.VGradient)
   B=ScalarParameter() {defaultValue=5, group=WPO, parameterName=Texture Scale Gradient Multiplier}
  ) {constB=5}
  n26 = ScalarParameter() {defaultValue=2, group=WPO, parameterName=Texture Scale Gradient Power}
  n27 = LinearInterpolate(
   A=ScalarParameter() {defaultValue=1, group=WPO, parameterName=ScaleV_Seam}
   B=ScalarParameter() {defaultValue=1, group=WPO, parameterName=ScaleV}
   Alpha=Clamp(None=Power(Base=@n25, Exp=@n26))
  )
  n28 = ComponentMask(None=VectorParameter() {group=WPO, parameterName=Pan Speed, defaultValue=<rgba 1 0 0 1>}.RGB) {g=true, r=true}
  n29 = Panner(Coordinate=AppendVector(A=Multiply(A=@n21.R, B=@n24), B=Multiply(A=@n21.G, B=@n27)), Speed=@n28) {bFractionalPart=true, speedX=0.025, speedY=-1}
  n30 = MaterialFunctionCall(Float2=Add(A=TextureCoordinate() {uTiling=2, vTiling=3}, B=ScalarParameter() {group=WPO, parameterName=SmokeOffset})) {materialFunction=&BreakOutFloat2Components}
  n31 = Panner(Coordinate=AppendVector(A=Multiply(A=@n30.R, B=@n24), B=Multiply(A=@n30.G, B=@n27)), Speed=@n28) {bFractionalPart=true, speedX=0.05, speedY=-1}
  n32 = Add(
   A=TextureSample(UVs=@n29) {samplerType=SAMPLERTYPE_LinearColor, texture=/Game/VisualThreatManager/StormVisuals/Textures/Clouds_VectorDisplacement}.G
   B=TextureSample(UVs=@n31) {samplerType=SAMPLERTYPE_LinearColor, texture=&Clouds_VectorDisplacement}.G
  )
  n33 = Power(Base=@n32, Exp=ScalarParameter() {defaultValue=2, group=WPO, parameterName=WPO_Detail_Contrast})
  n34 = VertexInterpolator(VS=AppendVector(A=@n32, B=@n33))
  n35 = ComponentMask(None=@n34.PS) {r=true}
  n36 = ScalarParameter() {defaultValue=2, group=Texture, parameterName=Texture_Detail_Contrast}
  n37 = TextureCoordinate()
  n38 = Add(
   A=Power(Base=ComponentMask(None=@n37) {g=true}, Exp=ScalarParameter() {defaultValue=2.5, parameterName=ColorBlendUVPower})
   B=ScalarParameter() {defaultValue=0.02, parameterName=ColorBlendUVAdd}
  )
  n39 = Add(A=@n38, B=ScalarParameter() {defaultValue=0.02, parameterName=ColorBlendUVAddHLOD})
  n40 = Power(
   Base=Saturate(None=MaterialProxyReplace(Realtime=@n38, MaterialProxy=@n39))
   Exp=ScalarParameter() {defaultValue=2.5, group=Glow, parameterName=GlowMaskPower}
  )
  n41 = TextureSample(UVs=@n29) {samplerType=SAMPLERTYPE_Normal, texture=&2ndInfectedMapNormal}
  n42 = TextureSample(UVs=@n31) {samplerType=SAMPLERTYPE_Normal, texture=&2ndInfectedMapNormal}
  n43 = DotProduct(
   A=Transform(None=MaterialFunctionCall(BaseNormal=@n41.RGB, AdditionalNormal=@n42.RGB) {materialFunction=&BlendAngleCorrectedNormals}.Result)
   B=Normalize(VectorInput=VectorParameter() {group=Glow, parameterName=GlowDirection, defaultValue=<rgba 0 0 -0.5 1>}.RGB)
  )
  n44 = ScalarParameter() {defaultValue=1, group=Glow, parameterName=InnerGlowMult}
  n45 = OneMinus(None=ComponentMask(None=@n34.PS) {g=true})
  n46 = Multiply(A=@n45, B=@n45)
  n47 = Add(
   A=Multiply(A=@n40, B=ScalarParameter() {defaultValue=0.5, group=Glow, parameterName=GlowMaskAdd}) {constB=0.5}
   B=Add(A=Saturate(None=@n43), B=Multiply(A=@n44, B=@n46))
  ) {constB=0.2}
  n48 = Power(Base=@n47, Exp=ScalarParameter() {defaultValue=3, group=Glow, parameterName=GlowMaskPower2}) {constExponent=3}
  n49 = Multiply(
   A=LinearInterpolate(
    A=VectorParameter() {group=Glow, parameterName=GlowColor2, defaultValue=<rgba 1 0.01 0 0>}.RGB
    B=VectorParameter() {group=Glow, parameterName=GlowColor1, defaultValue=<rgba 3 0.731 0 0>}.RGB
    Alpha=Saturate(None=@n48)
   )
   B=@n47
  )
  n50 = Add(
   A=Multiply(A=@n40, B=@n49)
   B=Multiply(A=@n49, B=ScalarParameter() {defaultValue=0.06, group=Glow, parameterName=UpperGlowMult}) {constB=0.1}
  )
  n51 = ScalarParameter() {defaultValue=1, group=Glow, parameterName=BaseGlow}
  n52 = Constant() {r=1.5}
  n53 = Constant() {r=1}
  Add(
   A=Multiply(A=Multiply(A=@n17, B=@n20), B=Power(Base=@n35, Exp=@n36))
   B=Multiply(
    A=Multiply(A=@n50, B=@n51)
    B=MaterialFunctionCall(Value_Morning=@n52, Value_Day=@n53, Value_Evening=@n52, Value_Night=@n53) {materialFunction=&Scalar_Time_of_Day}.Result
   )
  )
  n54 = Transform(None=Multiply(A=@n12.Result, B=VectorParameter() {parameterName=Offset, defaultValue=<rgba 150 150 150 0>}.RGB))
  n55 = VertexColor()
  n56 = LinearInterpolate(Alpha=@n33) {constA=-1, constB=2}
  n57 = VertexNormalWS()
  n58 = ScalarParameter() {defaultValue=1, group=WPO, parameterName=Deform_Seam}
  n59 = ScalarParameter() {defaultValue=1, group=WPO, parameterName=Deform}
  n60 = VertexColor()
  n61 = Multiply(
   A=Multiply(A=Multiply(A=@n56, B=@n57), B=LinearInterpolate(A=@n58, B=@n59, Alpha=@n23))
   B=Add(A=@n60.R, B=ScalarParameter() {group=WPO, parameterName=Vert Color Offset})
  )
  n62 = Multiply(A=@n57, B=Multiply(A=OneMinus(None=@n60.R), B=ScalarParameter() {group=WPO, parameterName=BottomWiden}))
  n63 = Constant3Vector()
  n64 = VectorParameter() {group=WPO_Wave, parameterName=TextureWPO_UVtiling, defaultValue=<rgba 1 2 0 0>}
  n65 = Panner(
   Coordinate=Multiply(A=TextureCoordinate(), B=ComponentMask(None=@n64.RGB) {g=true, r=true})
   Speed=ComponentMask(None=VectorParameter() {group=WPO_Wave, parameterName=TextureWPO_Speed, defaultValue=<rgba 1 2 0 0>}.RGB) {g=true, r=true}
  ) {speedY=0.15}
  n66 = TextureSample(UVs=@n65) {samplerType=SAMPLERTYPE_Normal, texture=&T_Weap_Distort_N_02}
  n67 = ComponentMask(None=@n66.RGB) {r=true}
  n68 = ScalarParameter() {defaultValue=25, group=WPO_Wave, parameterName=WPO_HeightX}
  n69 = ComponentMask(None=@n66.RGB) {g=true}
  n70 = ScalarParameter() {group=WPO_Wave, parameterName=WPO_HeightY}
  n71 = AppendVector(
   A=AppendVector(A=Multiply(A=@n67, B=@n68) {constB=0.5}, B=Multiply(A=@n69, B=@n70) {constB=0.5})
   B=Multiply(A=@n66.B, B=ScalarParameter() {group=WPO_Wave, parameterName=WPO_HeightZ}) {constB=0.6}
  )
  n72 = Constant() {r=0.5}
  n73 = SphereMask(
   A=MaterialFunctionCall() {materialFunction=&LinearGradient}.VGradient
   B=@n72
   Radius=@n72
  ) {hardnessPercent=0}
  n74 = Multiply(
   A=Transform(None=LinearInterpolate(A=@n63, B=@n71, Alpha=@n73))
   B=LinearInterpolate(
    A=ScalarParameter() {group=WPO_Wave, parameterName=WPO_Wave_Intensity_Close}
    B=ScalarParameter() {defaultValue=100, group=WPO_Wave, parameterName=WPO_Wave_Intenstiy}
   )
  ) {constB=80}
  n75 = Constant3Vector() {desc=Disable WPO anim in Low, constant=<rgba 0 0 20 0>}
  n76 = ShadingPathSwitch(Default=@n74, Deferred=@n74, Mobile=@n75) {desc=Migrated from FeatureLevelSwitch}
  Add(A=Add(A=Multiply(A=@n54, B=@n55.R), B=Add(A=@n61, B=@n62)), B=QualitySwitch(Default=@n76, Low=@n75, High=@n76, Medium=@n76))
  ComponentMask() {g=true}
  Multiply(A=@n54)
  n77 = PixelDepth()
  n78 = ScalarParameter() {defaultValue=-2199.7756, parameterName=BiasSorting_Far}
  n79 = SceneTexture() {sceneTextureId=PPI_CustomDepth}
  n80 = Constant()
  n81 = ComponentMask(None=TextureCoordinate()) {g=true}
  n82 = OneMinus(None=@n81)
  n83 = Clamp(None=Multiply(A=Multiply(A=@n81, B=@n82), B=ScalarParameter() {defaultValue=2, parameterName=ScaleCloudTipRamp}))
  n84 = ScalarParameter() {defaultValue=0.5, group=Fresnel, parameterName=TopAngleFade}
  n85 = ScalarParameter() {defaultValue=0.5, group=Fresnel, parameterName=BottomAngleFade}
  n86 = ComponentMask(None=TextureCoordinate()) {g=true}
  n87 = Saturate(None=MaterialFunctionCall("FadeSensitivity (optional)"=LinearInterpolate(A=@n84, B=@n85, Alpha=@n86)) {materialFunction=&FadeBasedOnViewAngle}.Result)
  n88 = ScalarParameter() {defaultValue=1200, group=Opacity, parameterName=CameraDistanceFade}
  n89 = Multiply(
   A=If(A=Add(A=@n77.R, B=@n78), B=ComponentMask(None=@n79.Color) {r=true}, "A > B"=@n80, "A == B"=@n80, "A < B"=Constant() {r=1})
   B=Multiply(
    A=Multiply(A=@n83, B=@n87)
    B=MaterialFunctionCall(Distance=@n88) {materialFunction=&Near_Camera_Fade}.Result
   )
  )
  n90 = DepthFade(Opacity=@n89, FadeDistance=ScalarParameter() {parameterName=Bottom_FadeDistance})
  Multiply(
   A=Multiply(A=ShadingPathSwitch(Default=@n90, Forward=@n89, Mobile=@n89), B=ScalarParameter() {defaultValue=1, group=Opacity, parameterName=OpacityMult})
   B=Constant() {r=1}
  )
  SphereMask(
   A=ComponentMask(None=WorldPosition().XYZ) {b=true}
   B=ComponentMask(None=ActorPositionWS()) {b=true}
   Radius=ScalarParameter() {defaultValue=50000, group=Glow, parameterName=GlowMaskRadius}
  ) {hardnessPercent=0}
  TextureSample(UVs=@n65) {texture=&T_Smooth_tile}
  VertexInterpolator(VS=@n33)
 ]
}
