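# Two-sided translucent lit material for distant Apollo seagulls, with a top-view texture scaled by Glow and vertex-shader-style motion.
# Sine wing flaps use UV side for inward direction and vertex color channels for masking, shape and timing variation.
# Flaps combine with axis rotation around the actor, or object position under a grass-actor switch.
# Default quality adds vector-noise drift; Low and Medium omit it.
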Material /Game/Effects/Fort_Effects/Materials/Environment/Apollo/M_Apollo_Seagulls_Far
graph {
 properties {
  bUsedWithInstancedStaticMeshes=true
  bUsedWithMeshDeformer=true
  bUsedWithMeshParticles=true
  bUsedWithSkeletalMesh=true
  blendMode=BLEND_Translucent
  pixelDepthOffsetMode=PDOM_Legacy
  twoSided=true
  usedShadingModels=Default Lit
 }
 nodes [
  n1 = Add(A=ScalarParameter() {defaultValue=1, parameterName=WingAnimSpeedMultiplier}, B=VertexColor().A)
  Add(A=ScalarParameter() {defaultValue=1, parameterName=WingAnimTimeOffset}, B=@n1)
  Constant4Vector() {constant=<rgba 1 1 1 0>}
  Multiply(
   A=TextureSample() {texture=/Game/Effects/Fort_Effects/Textures/Environments/Apollo/T_Apollo_Seagull_Top_D}.RGB
   B=ScalarParameter() {parameterName=Glow}
  )
  n2 = ComponentMask(None=TextureCoordinate()) {r=true}
  n3 = Multiply(
   A=Sign(None=Subtract(A=@n2) {constB=0.5})
   B=LinearInterpolate(
    A=ScalarParameter() {defaultValue=-0.389, parameterName=InwardVector}
    B=ScalarParameter() {defaultValue=-1.0138, parameterName=InwardVector_1}
    Alpha=VertexColor().G
   )
  )
  n4 = ScalarParameter() {defaultValue=0.134, parameterName=ForwardVector}
  n5 = AppendVector(A=@n3, B=@n4)
  n6 = AppendVector(A=@n4, B=@n3)
  n7 = ScalarParameter() {defaultValue=0.5823, parameterName=UPVector}
  n8 = ScalarParameter() {defaultValue=1.4967, parameterName=UPVector_1}
  n9 = VertexColor()
  n10 = Transform(None=AppendVector(A=StaticSwitchParameter(True=@n5, False=@n6) {parameterName=SwizzleXY}, B=LinearInterpolate(A=@n7, B=@n8, Alpha=@n9.B))) {transformSourceType=TRANSFORMSOURCE_Local}
  n11 = ScalarParameter() {defaultValue=15, parameterName=WindMovementDistance}
  n12 = Add(A=Time(), B=VertexColor().A)
  n13 = Multiply(A=VertexColor().G, B=ScalarParameter() {defaultValue=0.2, parameterName=WingOffset}) {constB=0.5}
  n14 = Multiply(A=ScalarParameter() {defaultValue=2, parameterName=SpeedWings}, B=VertexColor().A)
  n15 = Sine(None=Multiply(A=Add(A=@n12, B=@n13), B=Multiply(A=@n14, B=@n1)))
  n16 = ScalarParameter() {defaultValue=0.7853, parameterName=WindMovementBias}
  n17 = Multiply(
   A=Multiply(A=Multiply(A=@n10, B=@n11), B=Add(A=@n15, B=@n16))
   B=Add(A=VertexColor().R, B=ScalarParameter() {defaultValue=0.2286, parameterName=BodyMovement})
  )
  n18 = ScalarParameter() {parameterName=GrassActorWingMult}
  n19 = VectorParameter() {parameterName=RotationAxis}
  n20 = Time()
  n21 = ObjectPositionWS()
  n22 = StaticSwitchParameter(True=Add(A=@n20, B=ComponentMask(None=@n21) {b=true}), False=Frac(None=@n20)) {parameterName=UsewithGrassActor?}
  n23 = LinearInterpolate(
   A=ScalarParameter() {defaultValue=1, parameterName=RotationSpeed}
   B=ScalarParameter() {defaultValue=1, parameterName=RotationSpeed_Low}
   Alpha=VertexColor().A
  )
  n24 = ActorPositionWS()
  n25 = Add(
   A=StaticSwitchParameter(True=Multiply(A=@n17, B=@n18), False=@n17) {parameterName=UsewithGrassActor?}
   B=RotateAboutAxis(
    NormalizedRotationAxis=Transform(None=@n19.RGB) {transformSourceType=TRANSFORMSOURCE_Local}
    RotationAngle=Multiply(A=@n22, B=@n23)
    PivotPoint=StaticSwitchParameter(True=@n21, False=@n24) {parameterName=UsewithGrassActor?}
    Position=WorldPosition().XYZ
   )
  )
  n26 = VertexColor()
  n27 = LinearInterpolate(
   A=ScalarParameter() {defaultValue=50, parameterName=CurIntensity}
   B=ScalarParameter() {defaultValue=100, parameterName=CurIntensityHigh}
   Alpha=@n26.A
  )
  n28 = Subtract(A=WorldPosition().XYZ, B=ActorPositionWS())
  n29 = Multiply(A=Time(), B=ScalarParameter() {defaultValue=1, parameterName=CurlSpeed})
  n30 = ScalarParameter() {defaultValue=1000, parameterName=CurlScale}
  n31 = ScalarParameter() {defaultValue=500, parameterName=CurlScale_Low}
  n32 = VectorNoise(World Position=Divide(A=Subtract(A=@n28, B=@n29), B=LinearInterpolate(A=@n30, B=@n31, Alpha=@n26.A))) {noiseFunction=VNF_VectorALU}
  QualitySwitch(Default=Add(A=@n25, B=Multiply(A=@n27, B=@n32)), Low=@n25, Medium=@n25)
 ]
}
