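# Translucent unlit material for Apollo falling snowflakes, animated in the vertex shader on a mesh of sprites.
# World position offset builds and sizes sprites with near-camera shrink and minimum screen size, looping them across the object radius with sine drift and quality-switched curl noise.
# The snow progression mask fades opacity and pushes uncovered flakes 50000 units down.
# Flakes use randomized, rotating textures blendable to soft spheres, tinted by time of day and sun direction.
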Material /Game/Effects/Fort_Effects/Materials/Environment/Apollo/M_Apollo_SnowFlakes_MeshParticle_WT2020
graph {
 properties {
  bUsedWithInstancedStaticMeshes=true
  blendMode=BLEND_Translucent
  pixelDepthOffsetMode=PDOM_Legacy
  shadingModel=MSM_Unlit
  usedShadingModels=Unlit
 }
 nodes [
  n1 = StaticBoolParameter() {defaultValue=true, parameterName="IsFallingSnowFX (B)"}
  n2 = Clamp(None=Add(
    A=MaterialFunctionCall!(
     IsStaticSnow
     UseAlternateStaticSnowMask
     AlternateStaticSnowMask
     "UseRoad&&OcclusionMask"
     IsForGrassMeshToPushDown
     AdditionalMaskToRemoveSnow
     IsFallingSnowFX=@n1
     UseMinimapCoords
     IncreaseSnowUsingSlope
     UseSmoothBorder
     LerpHighM
     TemporalHeightLerpContrast
     IsLobbyIsland
     UseBR-IslandMask
     IsTree
     IsForRenderTarget
     UseSparkle
     UseSnowNormals
     UseAdditionalMaskToRemoveSnow
     IsTangentSpaceInput
     UseVertexNormalsForBasicSnow
     UseTangentSnowNormals
     Material
    ) {materialFunction=/Game/Athena/Apollo/Environments/Landscape/Materials/Biome/UniqueBiomeMaterialFunctions/MF_Apollo_Snow_Progression}.SimpleMask
    B=ScalarParameter() {parameterName=IsFrontEnd?}
   ))
  n3 = ScalarParameter() {defaultValue=4, group=Texture, parameterName="Time Speed (S)"}
  n4 = VertexColor()
  n5 = MaterialFunctionCall(
   UVs=TextureCoordinate()
   Rotation Center=VectorParameter() {parameterName="Rotation Center (V2)", defaultValue=<rgba 0.5 0.5 0 0>}.RGB
   "Rotation Angle (0-1)"=Add(A=Time(), B=Add(A=@n3, B=@n4.A))
  ) {materialFunction=&CustomRotator}
  n6 = MaterialFunctionCall(
   In=MaterialFunctionCall!(
    Random Picker=VertexColor().R
    Texture Input=TextureSampleParameter2D(UVs=@n5.Rotated Values) {parameterName=SnowTexture, texture=/Game/Weapons/WeaponSkins/Elemental/Textures/T_Snow_01}.RGB
    Option 1
    Option 2
    Option 3
   ) {materialFunction=/Game/Effects/Fort_Effects/Materials/_Functions/MF_RGB_Random_01}.Result
   Contrast=ScalarParameter() {group=Texture, parameterName="Contrast (S)"}
  ) {materialFunction=&CheapContrast}
  n7 = SphereMask(
   A=TextureCoordinate()
   B=Constant() {r=0.5}
   Radius=ScalarParameter() {defaultValue=0.5, group=Texture, parameterName=SnowFlake_Blurred_Radius}
   Hardness=ScalarParameter() {group=Texture, parameterName=Hardness}
  )
  n8 = ScalarParameter() {group=Texture, parameterName=SnowBlurriness}
  Multiply(A=ScalarParameter() {defaultValue=1, parameterName=Opacity}, B=Multiply(A=LinearInterpolate(A=@n6.Result, B=@n7, Alpha=@n8), B=@n2))
  n9 = ScalarParameter() {defaultValue=0.5, group=Texture, parameterName=ColorIntensity_SunSide_Mult}
  n10 = ScalarParameter() {defaultValue=0.35, group=Texture, parameterName=ColorIntensity_ShadowSide_Mult}
  n11 = CollectionParameter() {
   collection=&FortniteMaterialParameters
   parameterName=SunAndMoonModelDirectionalVector
  }
  n12 = Clamp(None=DotProduct(A=ComponentMask(None=@n11) {b=true, g=true, r=true}, B=ReflectionVectorWS()))
  n13 = VectorParameter() {group=TOD, parameterName="Color_Morning (V3)", defaultValue=<rgba 1 1 1 0>}
  n14 = VectorParameter() {group=TOD, parameterName="Color_Day (V3)", defaultValue=<rgba 1 1 1 0>}
  n15 = VectorParameter() {group=TOD, parameterName="Color_Evening (V3)", defaultValue=<rgba 1 1 1 0>}
  n16 = VectorParameter() {group=TOD, parameterName="Color_Night (V3)", defaultValue=<rgba 1 1 1 0>}
  Multiply(
   A=ScalarParameter() {defaultValue=1, group=Texture, parameterName=ColorIntensity}
   B=Multiply(
    A=LinearInterpolate(A=@n9, B=@n10, Alpha=@n12)
    B=MaterialFunctionCall(Color_Morning=@n13.RGB, Color_Day=@n14.RGB, Color_Evening=@n15.RGB, Color_Night=@n16.RGB) {materialFunction=&Color_Time_of_Day}.Result
   )
  )
  n17 = WorldPosition()
  n18 = ObjectPositionWS()
  n19 = ScalarParameter() {defaultValue=1, group=Movement, parameterName=Curl_Speed}
  n20 = Time()
  n21 = Divide(
   A=Subtract(A=Subtract(A=@n17.XYZ, B=@n18), B=Multiply(A=@n19, B=@n20))
   B=ScalarParameter() {defaultValue=1, group=Movement, parameterName=Curl_Scale}
  )
  n22 = Constant()
  n23 = ShadingPathSwitch(
   Default=Multiply(
    A=VectorNoise(World Position=@n21) {noiseFunction=VNF_VectorALU}
    B=ScalarParameter() {defaultValue=1, group=Movement, parameterName=Curl_Intensity}
   )
   Mobile=@n22
  ) {desc=Migrated from FeatureLevelSwitch}
  n24 = ShadingPathSwitch(Default=QualitySwitch(Default=@n23, Low=@n22, Medium=@n23), Forward=@n22)
  n25 = ScalarParameter() {defaultValue=25, group=WPO, parameterName=maxSize}
  n26 = Multiply(A=Multiply(A=ScalarParameter() {defaultValue=0.25, group=WPO, parameterName=MinSizeMult}, B=VertexColor().G), B=@n25)
  n27 = Time() {bOverride_Period=true, period=3}
  n28 = Time() {bOverride_Period=true, period=4}
  n29 = ComponentMask(None=TextureCoordinate() {coordinateIndex=1}) {g=true}
  n30 = Add(
   A=StaticSwitchParameter(True=Divide(A=@n27) {constB=3}, False=Divide(A=@n28) {constB=4}) {defaultValue=true, parameterName=UseFastSprites}
   B=Multiply(A=Subtract(A=@n29) {constB=0.5}) {constB=-2}
  )
  n31 = ConstantBiasScale(None=Sine(None=Add(A=@n30) {constB=-0.25}))
  n32 = LinearInterpolate(A=@n26, B=@n25, Alpha=Power(Base=LinearInterpolate(Alpha=@n31)) {constExponent=0.6})
  n33 = ViewSize()
  n34 = ScalarParameter() {defaultValue=0.08, group=WPO, parameterName=MinSizeatDistance_43_ScreenRatio}
  n35 = MaterialFunctionCall() {materialFunction=&ScreenResolution}
  n36 = ScalarParameter() {defaultValue=1.33, parameterName=ScreenRes_Mininum}
  n37 = ScalarParameter() {defaultValue=2, parameterName=ScreenRes_Max}
  n38 = Clamp(
   None=Divide(A=ComponentMask(None=@n35.Visible Resolution) {r=true}, B=ComponentMask(None=@n35.Visible Resolution) {g=true})
   Min=@n36
   Max=@n37
  ) {maxDefault=2, minDefault=1.33}
  n39 = Saturate(None=Divide(A=Subtract(A=@n38, B=@n36), B=Subtract(A=@n37, B=@n36)))
  n40 = ShadingPathSwitch(
   Default=ScalarParameter() {defaultValue=0.08, group=WPO, parameterName=MinSizeatDistance_Full_ScreenRatio}
   Forward=Divide(
    A=ScalarParameter() {defaultValue=0.08, group=WPO, parameterName=MinSizeatDistance_Full_ScreenRatio_Switch}
    B=ScalarParameter() {defaultValue=0.75, parameterName=SizeIncreaser_NintendoSwitch}
   )
  )
  n41 = Divide(A=PixelDepth().R, B=Multiply(A=ComponentMask(None=@n33) {g=true}, B=LinearInterpolate(A=@n34, B=@n40, Alpha=@n39)))
  n42 = Multiply(
   A=MaterialFunctionCall(Distance=ScalarParameter() {defaultValue=500, group=WPO, parameterName=NearCamShrinkDist}) {materialFunction=&Near_Camera_Fade}.Result
   B=LinearInterpolate(A=Constant() {r=1}, B=Add(A=@n26, B=@n41), Alpha=ScalarParameter() {defaultValue=1, group=WPO, parameterName=UseMinimumSizeOnScreen?})
  )
  n43 = Multiply(A=ObjectRadius(), B=Constant() {r=0.5}) {constB=0.5}
  n44 = Multiply(
   A=LinearInterpolate(A=Multiply(A=@n43) {constB=-1}, B=@n43, Alpha=Frac(None=@n30))
   B=ScalarParameter() {defaultValue=1, group=Movement, parameterName=OverallSpeed}
  )
  n45 = VectorParameter() {group=Movement, parameterName=OverallDirection, defaultValue=<rgba 1 0 0 1>}
  n46 = VertexColor()
  n47 = ScalarParameter() {group=Movement, parameterName=OveralDirection_Spread}
  n48 = LinearInterpolate(A=@n45.RGB, B=OneMinus(None=@n45.RGB), Alpha=Clamp(None=Multiply(A=@n46.G, B=@n47)))
  n49 = Multiply(A=Time(), B=ScalarParameter() {defaultValue=0.001, group=Movement, parameterName=VarianceSpeed}) {constB=0.05}
  n50 = ComponentMask(None=TextureCoordinate() {coordinateIndex=1}) {r=true}
  n51 = Add(
   A=Multiply(A=Time(), B=ScalarParameter() {defaultValue=0.4, group=Movement, parameterName=DriftSpeed})
   B=Sine(None=Add(A=@n49, B=@n50)) {period=0.17}
  )
  n52 = AppendVector(A=AppendVector(A=Sine(None=@n51) {period=0.56}, B=Sine(None=@n51) {period=0.7}), B=Sine(None=@n51) {period=0.82})
  n53 = LinearInterpolate(
   A=ScalarParameter() {defaultValue=1, group=Movement, parameterName=MinHorizontalDrift}
   B=ScalarParameter() {defaultValue=1, group=Movement, parameterName=MaxHorizontalDrift}
   Alpha=Sine(None=Add(A=@n49, B=@n50)) {period=0.14}
  )
  n54 = Add(
   A=MaterialFunctionCall!(Center Location=WorldPosition().XYZ, 0-1 UVs, XY scale=Multiply(A=@n32, B=@n42), OverridePivot, "Normals (Optional)") {materialFunction=&Sprite}.World position offset
   B=Add(A=Multiply(A=@n52, B=@n53), B=Multiply(A=@n44, B=@n48))
  )
  n55 = Add(A=@n2, B=ScalarParameter() {defaultValue=0.2, parameterName=SubtractWPOMask_Value})
  Reroute(None=LinearInterpolate(A=Constant3Vector() {constant=<rgba 0 0 -50000 0>}, B=Add(A=@n54, B=@n24), Alpha=Floor(None=@n55)))
 ]
}
