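# Single Layer Water material for the Apollo waterfall that blends a panning foam layer over Fortnite river flowmap water.
# Water velocity and depth lerp from fixed parameters to WaterBodyData along UV V, then feed FortWaterRiverFlowmaps.
# Foam combines noise-distorted, power-warped UV panners into tinted color and flattened blended normals, layered in through V-band masks.
# Scrolling distortion textures build world-position-offset deformation and waves; quality and shading-path switches simplify specular and scattering.
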Material /Game/Effects/Fort_Effects/Materials/Environment/Apollo/M_Apollo_WaterFall_FoamBlend
graph {
 properties {
  bTangentSpaceNormal=false
  pixelDepthOffsetMode=PDOM_Legacy
  refractionMethod=RM_PixelNormalOffset
  shadingModel=MSM_SingleLayerWater
  usedShadingModels=SingleLayerWater
 }
 nodes [
  Add(B=ScalarParameter() {parameterName=Refraction}) {constA=1}
  Constant3Vector() {constant=<rgba 0 0 1 0>}
  n1 = ComponentMask(None=TextureCoordinate()) {g=true}
  LinearInterpolate(A=@n1)
  LinearInterpolate() {constA=-1, constB=2}
  n2 = ComponentMask(None=VectorParameter() {group=Added Foam, parameterName=FoamTileMult, defaultValue=<rgba 1 5 0 0>}.RGB) {g=true, r=true}
  n3 = TextureCoordinate() {uTiling=2.8, vTiling=0.7}
  n4 = ComponentMask(None=@n3) {g=true}
  n5 = ScalarParameter() {defaultValue=0.781, group=Added Foam, parameterName=EndTextureTile}
  n6 = Multiply(A=@n2, B=AppendVector(A=ComponentMask(None=@n3) {r=true}, B=Power(Base=@n4, Exp=@n5)))
  n7 = ComponentMask(None=VectorParameter() {parameterName=FoamSpeed_01, defaultValue=<rgba 0 -1 0 0>}.RGB) {g=true, r=true}
  n8 = Time()
  n9 = Panner(
   Coordinate=Multiply(A=TextureCoordinate(), B=ScalarParameter() {defaultValue=0.7372, group=Added Foam, parameterName=UVDistortTilingA})
   Time=@n8
   Speed=@n7
  ) {speedY=0.5}
  n10 = ComponentMask(None=VectorParameter() {parameterName=FoamSpeed_02, defaultValue=<rgba 0 -1 0 0>}.RGB) {g=true, r=true}
  n11 = Panner(
   Coordinate=Multiply(A=TextureCoordinate(), B=ScalarParameter() {defaultValue=0.8, group=Added Foam, parameterName=UVDistortTilingB})
   Time=@n8
   Speed=@n10
  ) {speedX=0.125, speedY=0.3}
  n12 = LinearInterpolate(Alpha=ComponentMask(None=TextureCoordinate()) {g=true}) {constB=1.5}
  n13 = LinearInterpolate(
   A=Multiply(
    A=TextureSample(UVs=@n9) {texture=/Game/Global/Textures/Default/Noise/T_Tiling_Noise_01}.G
    B=ScalarParameter() {group=Added Foam, parameterName=UVDistortAmountA}
   )
   B=Multiply(
    A=TextureSample(UVs=@n11) {texture=&T_Tiling_Noise_01}.G
    B=ScalarParameter() {group=Added Foam, parameterName=UVDistortAmountB}
   )
   Alpha=Clamp(None=OneMinus(None=@n12))
  )
  n14 = Panner(Coordinate=Add(A=OneMinus(None=@n6), B=@n13), Speed=@n10) {speedY=-1}
  n15 = ComponentMask(None=VectorParameter() {group=Added Foam, parameterName=AddedFoam_WPO_Pan Speed, defaultValue=<rgba 0 0.1 0 1>}.RGB) {g=true, r=true}
  n16 = TextureSample(UVs=Panner(Coordinate=@n14, Speed=@n15) {bFractionalPart=true, speedX=0.025, speedY=-1}) {samplerType=SAMPLERTYPE_Normal, texture=&T_Weap_Distort_N_02}
  n17 = TextureCoordinate() {uTiling=2, vTiling=0.6}
  n18 = ComponentMask(None=@n17) {g=true}
  n19 = ScalarParameter() {defaultValue=0.781, group=Added Foam, parameterName=EndTextureTile}
  n20 = Multiply(A=@n2, B=AppendVector(A=ComponentMask(None=@n17) {r=true}, B=Power(Base=@n18, Exp=@n19)))
  n21 = Panner(
   Coordinate=Add(A=OneMinus(None=@n20), B=@n13)
   Speed=ComponentMask(None=VectorParameter() {parameterName=FoamSpeed_03, defaultValue=<rgba 0 -1 0 0>}.RGB) {g=true, r=true}
  ) {speedY=-1}
  n22 = TextureSample(UVs=Panner(Coordinate=@n21, Speed=@n15) {bFractionalPart=true, speedX=-0.05, speedY=-1}) {samplerType=SAMPLERTYPE_Normal, texture=&T_Weap_Distort_N_02}
  n23 = Multiply(A=@n16.RGB, B=@n22.RGB)
  n24 = VertexNormalWS()
  n25 = SetMaterialAttributes!(
   MaterialAttributes
   Base Color=VectorParameter() {parameterName=BaseColor}.RGB
   Specular=ScalarParameter() {defaultValue=0.225, group=Water Shading Model, parameterName=Specular}
   Roughness=ScalarParameter() {defaultValue=0.1, group=Water Shading Model, parameterName=Roughness}
   Opacity=ScalarParameter() {group=Water Shading Model, parameterName=Surface Opacity}
   Normal=Constant3Vector() {constant=<rgba 0 0 1 0>}
   World Position Offset
  ) {attributeSetTypes [
    #1
    #2
    #3
    #4
    #5
    #6
   ]}
  n26 = VectorParameter() {parameterName=Water_Velocity_Top, defaultValue=<rgba -800 0 600 256>}
  n27 = MaterialFunctionCall!(Texture, Position) {materialFunction=/FortWater/Materials/Functions/WaterBodyData}
  n28 = ComponentMask(None=TextureCoordinate()) {g=true}
  n29 = ScalarParameter() {defaultValue=0.65, group=Added Foam, parameterName=FoamBlendMask_Back_Radius}
  n30 = Saturate(None=Divide(A=Subtract(A=@n28, B=@n29), B=ScalarParameter() {defaultValue=0.05, group=Added Foam, parameterName=FoamBlendMask_Back_Divide}))
  n31 = MaterialFunctionCall!(
   Attributes=@n25
   Transition
   Sim Foam Erasure
   Velocity=LinearInterpolate(A=@n26.RGB, B=@n27.VelocityAndLength, Alpha=@n30)
   Far Mask Override
   Depth=LinearInterpolate(A=@n26.A, B=@n27.Water Depth, Alpha=@n30)
  ) {materialFunction=/FortWater/Materials/Functions/FortWaterRiverFlowmaps}
  n32 = VectorParameter() {group=Added Foam, parameterName=FoamColor_WaterUnderFoamColor_Bottom, defaultValue=<rgba 0.1089 0.2253 0.2396 1>}
  n33 = VectorParameter() {group=Added Foam, parameterName=FoamColor_WaterUnderFoamColor_Top, defaultValue=<rgba 0.0805 0.1665 0.1771 1>}
  n34 = Clamp(None=Power(Base=@n1, Exp=ScalarParameter() {defaultValue=1, group=Added Foam, parameterName=FoamExp}) {constExponent=0.5})
  n35 = TextureSample(UVs=@n21) {samplerType=SAMPLERTYPE_LinearColor, texture=/Game/Global/Textures/Default/Caustics/03_WaterFoam_AccentSwirls_H}
  n36 = TextureCoordinate()
  n37 = Power(Base=OneMinus(None=ComponentMask(None=@n36) {g=true}), Exp=ScalarParameter() {defaultValue=1, group=Added Foam, parameterName=FoamExp}) {constExponent=0.5}
  n38 = TextureCoordinate() {vTiling=0.4}
  n39 = ComponentMask(None=@n38) {g=true}
  n40 = ScalarParameter() {defaultValue=0.781, group=Added Foam, parameterName=EndTextureTile}
  n41 = Multiply(A=AppendVector(A=ComponentMask(None=@n38) {r=true}, B=Power(Base=@n39, Exp=@n40)), B=@n2)
  n42 = Panner(Coordinate=Add(A=OneMinus(None=@n41), B=@n13), Speed=@n7) {speedY=-1}
  n43 = Multiply(
   A=@n37
   B=ComponentMask(None=TextureSample(UVs=@n42) {samplerType=SAMPLERTYPE_LinearColor, texture=&03_WaterFoam_AccentSwirls_H}.G) {g=true}
  )
  n44 = Add(
   A=@n43
   B=ComponentMask(None=TextureSample(UVs=@n14) {samplerType=SAMPLERTYPE_LinearColor, texture=&03_WaterFoam_AccentSwirls_H}.G) {g=true}
  )
  n45 = Clamp(None=@n37)
  n46 = Multiply(
   A=LinearInterpolate(
    A=VectorParameter() {group=Added Foam, parameterName=FoamColor_Bottom, defaultValue=<rgba 0.332 0.3542 0.3542 1>}.RGB
    B=VectorParameter() {group=Added Foam, parameterName=FoamColorTop, defaultValue=<rgba 0.1728 0.276 0.2728 1>}.RGB
    Alpha=@n34
   )
   B=LinearInterpolate(A=ComponentMask(None=@n35.G) {g=true}, B=LinearInterpolate(A=@n43, B=@n44, Alpha=@n45), Alpha=@n34) {constAlpha=0.4}
  )
  n47 = TextureSample(UVs=@n21) {samplerType=SAMPLERTYPE_Normal, texture=/Game/Athena/Apollo/Water/Textures/02_Rounded/02_WaterFoam_Veroni_base_N}
  n48 = TextureSample(UVs=@n14) {samplerType=SAMPLERTYPE_Normal, texture=&02_WaterFoam_Veroni_base_N}
  n49 = Transform(None=MaterialFunctionCall(
    Normal=MaterialFunctionCall(BaseNormal=@n47.RGB, AdditionalNormal=@n48.RGB) {materialFunction=&BlendAngleCorrectedNormals}.Result
    Flatness=ScalarParameter() {defaultValue=0.8, parameterName=FoamNormalFlatten}
   ) {materialFunction=&FlattenNormal}.Result)
  n50 = TextureCoordinate()
  n51 = Power(Base=OneMinus(None=ComponentMask(None=@n50) {g=true}), Exp=ScalarParameter() {defaultValue=5, parameterName=BottomNormalTranistion})
  n52 = SetMaterialAttributes(
   Base Color=LinearInterpolate(A=LinearInterpolate(A=@n32.RGB, B=@n33.RGB, Alpha=@n34), Alpha=Clamp(None=@n46))
   Specular=ScalarParameter() {defaultValue=0.225, group=Added Foam, parameterName=Specular2}
   Roughness=ScalarParameter() {defaultValue=1, group=Added Foam, parameterName=Roughness2}
   Opacity=ScalarParameter() {defaultValue=1, group=Added Foam, parameterName=Surface Opacity2}
   Normal=LinearInterpolate(A=Normalize(VectorInput=@n49), B=Constant3Vector() {constant=<rgba 0 0 1 0>}, Alpha=Saturate(None=@n51))
  ) {attributeSetTypes [
    #1
    #2
    #3
    #4
    #5
    #6
   ]}
  n53 = ComponentMask(None=TextureCoordinate()) {g=true}
  n54 = ScalarParameter() {defaultValue=0.1, group=Added Foam, parameterName=FoamBlendMask_WaterLevel_Start}
  n55 = Saturate(None=Divide(A=Subtract(A=@n53, B=@n54), B=ScalarParameter() {defaultValue=0.1, group=Added Foam, parameterName=FoamBlendMask_WaterLevel_Width}))
  n56 = ComponentMask(None=TextureCoordinate()) {g=true}
  n57 = ScalarParameter() {defaultValue=0.55, group=Added Foam, parameterName=FoamBlendMask_Front_Start}
  n58 = Saturate(None=Divide(A=Subtract(A=@n56, B=@n57), B=ScalarParameter() {defaultValue=-0.05, group=Added Foam, parameterName=FoamBlendMask_Front_Width}))
  n59 = GetMaterialAttributes(None=MaterialFunctionCall!(Base Material=@n31.Attributes, Top Material=@n52, Alpha=Multiply(A=@n55, B=@n58)) {materialFunction=/Engine/Functions/MaterialLayerFunctions/MatLayerBlend_Standard}.Blended Material) {attributeGetTypes [
    #1
    #2
    #3
    #4
    #5
    #6
   ]}
  LinearInterpolate(
   A=Multiply(A=Multiply(A=@n23, B=@n24), B=ScalarParameter() {defaultValue=1, group=Added Foam, parameterName=AddedFoam_Deform})
   B=@n59.World Position Offset
   Alpha=@n30
  )
  n60 = ComponentMask(None=VectorParameter() {group=Added Foam, parameterName=AddedFoam_WPO_Tiling, defaultValue=<rgba 1 1 0 0>}.RGB) {g=true, r=true}
  Multiply(A=@n60, B=TextureCoordinate() {uTiling=2, vTiling=4})
  Multiply(A=TextureCoordinate() {vTiling=2}, B=@n60)
  n61 = Constant3Vector()
  n62 = TextureCoordinate()
  n63 = ComponentMask(None=VectorParameter() {group=WPO_Wave, parameterName=TextureWPO_UVtiling, defaultValue=<rgba 25 25 0 0>}.RGB) {g=true, r=true}
  n64 = VectorParameter() {group=WPO_Wave, parameterName=TextureWPO_Speed, defaultValue=<rgba 1 1 0 0>}
  n65 = TextureSample(UVs=Panner(Coordinate=Multiply(A=@n62, B=@n63), Speed=ComponentMask(None=@n64.RGB) {g=true, r=true}) {speedY=0.15}) {samplerType=SAMPLERTYPE_Normal, texture=&T_Weap_Distort_N_02}
  n66 = ComponentMask(None=@n65.RGB) {r=true}
  n67 = ScalarParameter() {defaultValue=5, group=WPO_Wave, parameterName=WPO_HeightX}
  n68 = ComponentMask(None=@n65.RGB) {g=true}
  n69 = ScalarParameter() {defaultValue=5, group=WPO_Wave, parameterName=WPO_HeightY}
  n70 = AppendVector(
   A=AppendVector(A=Multiply(A=@n66, B=@n67) {constB=0.5}, B=Multiply(A=@n68, B=@n69) {constB=0.5})
   B=Multiply(A=@n65.B, B=ScalarParameter() {defaultValue=1, group=WPO_Wave, parameterName=WPO_HeightZ}) {constB=0.6}
  )
  n71 = Constant() {r=0.5}
  n72 = SphereMask(
   A=MaterialFunctionCall() {materialFunction=&LinearGradient}.VGradient
   B=@n71
   Radius=@n71
  ) {hardnessPercent=0}
  Multiply(
   A=Transform(None=LinearInterpolate(A=@n61, B=@n70, Alpha=@n72))
   B=LinearInterpolate(
    A=ScalarParameter() {group=WPO_Wave, parameterName=WPO_Wave_Intensity_Close}
    B=ScalarParameter() {defaultValue=25, group=WPO_Wave, parameterName=WPO_Wave_Intenstiy}
   )
  ) {constB=80}
  Power(Base=Add(A=@n16.RGB, B=@n22.RGB)) {constExponent=5}
  Reroute(None=@n59.Base Color)
  Reroute(None=@n59.Normal)
  Reroute(None=@n59.Opacity)
  Reroute(None=@n59.Roughness)
  n73 = Constant() {r=0.4}
  n74 = QualitySwitch(Default=@n59.Specular, Low=@n73, High=@n59.Specular, Medium=@n73)
  ShadingPathSwitch(Default=@n74, Deferred=@n74, Forward=@n73, Mobile=@n73)
  n75 = Constant()
  n76 = QualitySwitch(Default=@n31.Scattering, Low=@n75, Medium=@n75)
  n77 = VectorParameter() {group=Water Shading Model, parameterName=Scattering, defaultValue=<rgba 0.1 0.1 0.1 0.5>}
  n78 = Divide(A=@n77.A) {constB=100}
  n79 = VectorParameter() {group=Water Shading Model, parameterName=Absorption, defaultValue=<rgba 1 8 20 100>}
  SingleLayerWaterMaterialOutput!(
   ScatteringCoefficients=Add(A=ShadingPathSwitch(Default=@n76, Deferred=@n76, Forward=@n75, Mobile=@n75), B=Multiply(A=@n77.RGB, B=@n78))
   AbsorptionCoefficients=Divide(B=Multiply(A=@n79.RGB, B=@n79.RGB))
   PhaseG=ScalarParameter() {defaultValue=0.3, group=Water Shading Model, parameterName=Anisotropy}
   ColorScaleBehindWater
  )
  VectorParameter() {parameterName=Water_Velocity, defaultValue=<rgba -800 0 600 256>}
 ]
}
