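# Two-sided Single Layer Water material that blends waterfall foam into ocean water.
# The water layer uses vertex blue to blend flattened flow-map normals, directed relative to a center-location parameter, with FortWater two-texture panner normals.
# The foam layer pans noise-distorted swirl textures with V-gradient coloring and vertex-green opacity.
# A height lerp driven by vertex red and foam texture blends both layers, with cheaper specular and refraction fallbacks.
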Material /Game/Effects/Fort_Effects/Materials/Environment/Apollo/M_Apollo_WaterFall_FoamBlend_Ocean
graph {
 properties {
  allowTranslucentCustomDepthWrites=true
  bTangentSpaceNormal=false
  pixelDepthOffsetMode=PDOM_Legacy
  refractionMethod=RM_PixelNormalOffset
  shadingModel=MSM_SingleLayerWater
  twoSided=true
  usedShadingModels=SingleLayerWater
 }
 nodes [
  n1 = TextureObject() {samplerType=SAMPLERTYPE_LinearColor, texture=/Game/Global/Textures/Default/Tileable_Patterns/T_Water_TilingNormal_With_Height_02_Softened}
  n2 = VectorParameter() {parameterName=CenterObjectLocation, defaultValue=<rgba 384 -128 -3072 1>}
  n3 = Subtract(A=ComponentMask(None=WorldPosition().XYZ) {g=true, r=true}, B=AppendVector(A=@n2.R, B=@n2.R))
  n4 = WorldPosition()
  n5 = MaterialFunctionCall!(
   Diffuse
   Normal Map=@n1
   "Flow Vector Map (see tooltip)"=Multiply(A=Normalize(VectorInput=@n3), B=Constant2Vector() {g=1, r=-1})
   UV texture 2 offset
   "Time (see tooltip)"=Multiply(A=Time() {bOverride_Period=true, period=5}, B=Constant() {r=0.4})
   UVs=Divide(A=ComponentMask(None=@n4.XYZ) {g=true, r=true}, B=Constant() {r=2048})
  ) {materialFunction=/Engine/Functions/Engine_MaterialFunctions02/Texturing/FlowMaps_Simple}
  n6 = Constant() {r=0.95}
  n7 = MaterialFunctionCall!(
   Normal
   WPO=Constant()
   Far Mask Override=MaterialFunctionCall(
    Fade Length=ScalarParameter() {defaultValue=15000, group=Distance Blending, parameterName=Far Fade Length}
    Fade Offset=ScalarParameter() {defaultValue=7500, group=Distance Blending, parameterName=Far Fade Start}
   ) {materialFunction=&CameraDepthFade}.Result
   Additional Mask
   LowDetailBiome_Mask=Constant()
  ) {materialFunction=/FortWater/Materials/Functions/WaterTwoTexturePanner}
  n8 = SetMaterialAttributes!(
   MaterialAttributes
   Base Color=VectorParameter() {parameterName=BaseColor}.RGB
   Specular=Constant() {r=0.225}
   Roughness=Constant() {r=0.03}
   Opacity=Constant()
   Normal=LinearInterpolate(
    A=MaterialFunctionCall(Normal=@n5.Normal, Flatness=@n6) {materialFunction=&FlattenNormal}.Result
    B=@n7.Normal
    Alpha=VertexColor().B
   )
   World Position Offset
  ) {attributeSetTypes [
    #1
    #2
    #3
    #4
    #5
    #6
   ]}
  n9 = VertexColor()
  n10 = ComponentMask(None=VectorParameter() {group=Added Foam, parameterName=FoamTileMult, defaultValue=<rgba 1 5 0 0>}.RGB) {g=true, r=true}
  n11 = TextureCoordinate() {uTiling=2, vTiling=0.6}
  n12 = AppendVector(
   A=ComponentMask(None=@n11) {r=true}
   B=Power(Base=ComponentMask(None=@n11) {g=true}, Exp=ScalarParameter() {defaultValue=0.781, group=Added Foam, parameterName=EndTextureTile})
  )
  n13 = ComponentMask(None=VectorParameter() {parameterName=FoamSpeed_01, defaultValue=<rgba 0 -1 0 0>}.RGB) {g=true, r=true}
  n14 = Time()
  n15 = Panner(
   Coordinate=Multiply(A=TextureCoordinate(), B=ScalarParameter() {defaultValue=0.7372, group=Added Foam, parameterName=UVDistortTilingA})
   Time=@n14
   Speed=@n13
  ) {speedY=0.5}
  n16 = ComponentMask(None=VectorParameter() {parameterName=FoamSpeed_02, defaultValue=<rgba 0 -1 0 0>}.RGB) {g=true, r=true}
  n17 = Panner(
   Coordinate=Multiply(A=TextureCoordinate(), B=ScalarParameter() {defaultValue=0.8, group=Added Foam, parameterName=UVDistortTilingB})
   Time=@n14
   Speed=@n16
  ) {speedX=0.125, speedY=0.3}
  n18 = LinearInterpolate(Alpha=ComponentMask(None=TextureCoordinate()) {g=true}) {constB=1.5}
  n19 = LinearInterpolate(
   A=Multiply(
    A=TextureSample(UVs=@n15) {texture=/Game/Global/Textures/Default/Noise/T_Tiling_Noise_01}.G
    B=ScalarParameter() {group=Added Foam, parameterName=UVDistortAmountA}
   )
   B=Multiply(
    A=TextureSample(UVs=@n17) {texture=&T_Tiling_Noise_01}.G
    B=ScalarParameter() {group=Added Foam, parameterName=UVDistortAmountB}
   )
   Alpha=Clamp(None=OneMinus(None=@n18))
  )
  n20 = Add(A=OneMinus(None=Multiply(A=@n10, B=@n12)), B=@n19)
  n21 = ComponentMask(None=VectorParameter() {parameterName=FoamSpeed_03, defaultValue=<rgba 0 -1 0 0>}.RGB) {g=true, r=true}
  n22 = Panner(Coordinate=@n20, Speed=@n21) {speedY=-1}
  n23 = VectorParameter() {group=Added Foam, parameterName=FoamTile_Mult3, defaultValue=<rgba 0.5 0.5 0 0>}
  n24 = VectorParameter() {group=Added Foam, parameterName=FoamSpeed_Mult3, defaultValue=<rgba 1 1 0 0>}
  n25 = Panner(
   Coordinate=Multiply(A=@n20, B=ComponentMask(None=@n23.RGB) {g=true, r=true})
   Speed=Multiply(A=@n21, B=ComponentMask(None=@n24.RGB) {g=true, r=true})
  ) {speedY=-1}
  n26 = Multiply(
   A=@n20
   B=ComponentMask(None=VectorParameter() {group=Added Foam, parameterName=Foam_mask_Tile_Mult, defaultValue=<rgba 0.5 0.5 0 0>}.RGB) {g=true, r=true}
  )
  n27 = Multiply(
   A=@n21
   B=ComponentMask(None=VectorParameter() {group=Added Foam, parameterName=FoamSpeed_Mask_Mult, defaultValue=<rgba 1 1 0 0>}.RGB) {g=true, r=true}
  )
  n28 = Power(Base=TextureSample(UVs=Panner(Coordinate=@n26, Speed=@n27) {speedY=-1}) {texture=&T_Tiling_Noise_01}.G)
  n29 = ComponentMask(None=LinearInterpolate(
    A=TextureSample(UVs=@n22) {samplerType=SAMPLERTYPE_LinearColor, texture=/Game/Global/Textures/Default/Caustics/03_WaterFoam_AccentSwirls_H}.G
    B=TextureSample(UVs=@n25) {samplerType=SAMPLERTYPE_LinearColor, texture=&03_WaterFoam_AccentSwirls_H}.G
    Alpha=@n28
   )) {g=true}
  n30 = Saturate(None=Multiply(A=@n29, B=Constant() {r=0.4}))
  n31 = VectorParameter() {group=Added Foam, parameterName=FoamColor_WaterUnderFoamColor_Bottom, defaultValue=<rgba 0.1089 0.2253 0.2396 1>}
  n32 = VectorParameter() {group=Added Foam, parameterName=FoamColor_WaterUnderFoamColor_Top, defaultValue=<rgba 0.0805 0.1665 0.1771 1>}
  n33 = TextureCoordinate()
  n34 = Clamp(None=Power(Base=ComponentMask(None=@n33) {g=true}, Exp=ScalarParameter() {defaultValue=1, group=Added Foam, parameterName=FoamExp}) {constExponent=0.5})
  n35 = LinearInterpolate(A=@n31.RGB, B=@n32.RGB, Alpha=@n34)
  n36 = LinearInterpolate(
   A=VectorParameter() {group=Added Foam, parameterName=FoamColor_Bottom, defaultValue=<rgba 0.332 0.3542 0.3542 1>}.RGB
   B=VectorParameter() {group=Added Foam, parameterName=FoamColorTop, defaultValue=<rgba 0.1728 0.276 0.2728 1>}.RGB
   Alpha=@n34
  )
  n37 = TextureCoordinate()
  n38 = Power(Base=OneMinus(None=ComponentMask(None=@n37) {g=true}), Exp=ScalarParameter() {defaultValue=1, group=Added Foam, parameterName=FoamExp}) {constExponent=0.5}
  n39 = TextureCoordinate() {vTiling=0.4}
  n40 = ComponentMask(None=@n39) {g=true}
  n41 = ScalarParameter() {defaultValue=0.781, group=Added Foam, parameterName=EndTextureTile}
  n42 = Multiply(A=AppendVector(A=ComponentMask(None=@n39) {r=true}, B=Power(Base=@n40, Exp=@n41)), B=@n10)
  n43 = Panner(Coordinate=Add(A=OneMinus(None=@n42), B=@n19), Speed=@n13) {speedY=-1}
  n44 = Multiply(
   A=@n38
   B=ComponentMask(None=TextureSample(UVs=@n43) {samplerType=SAMPLERTYPE_LinearColor, texture=&03_WaterFoam_AccentSwirls_H}.G) {g=true}
  )
  n45 = TextureCoordinate() {uTiling=2.8, vTiling=0.7}
  n46 = AppendVector(
   A=ComponentMask(None=@n45) {r=true}
   B=Power(Base=ComponentMask(None=@n45) {g=true}, Exp=ScalarParameter() {defaultValue=0.781, group=Added Foam, parameterName=EndTextureTile})
  )
  n47 = Add(A=OneMinus(None=Multiply(A=@n10, B=@n46)), B=@n19)
  n48 = Panner(Coordinate=@n47, Speed=@n16) {speedY=-1}
  n49 = TextureSample(UVs=@n48) {samplerType=SAMPLERTYPE_LinearColor, texture=&03_WaterFoam_AccentSwirls_H}
  n50 = ComponentMask(None=VectorParameter() {group=Added Foam, parameterName=FoamTile_Mult2, defaultValue=<rgba 0.5 0.5 0 0>}.RGB) {g=true, r=true}
  n51 = ComponentMask(None=VectorParameter() {group=Added Foam, parameterName=FoamSpeed_Mult2, defaultValue=<rgba 1 1 0 0>}.RGB) {g=true, r=true}
  n52 = TextureSample(UVs=Panner(Coordinate=Multiply(A=@n47, B=@n50), Speed=Multiply(A=@n16, B=@n51)) {speedY=-1}) {samplerType=SAMPLERTYPE_LinearColor, texture=&03_WaterFoam_AccentSwirls_H}
  n53 = ComponentMask(None=FeatureLevelSwitch(Default=LinearInterpolate(A=@n49.G, B=@n52.G, Alpha=@n28), ES2=@n49.G, ES3_1=@n49.G)) {g=true}
  n54 = LinearInterpolate(A=@n29, B=LinearInterpolate(A=@n44, B=Add(A=@n44, B=@n53), Alpha=Clamp(None=@n38)), Alpha=@n34) {constAlpha=0.4}
  n55 = LinearInterpolate(A=@n35, Alpha=Clamp(None=Multiply(A=@n36, B=@n54)))
  n56 = MaterialFunctionCall(
   Normal=MaterialFunctionCall(
    BaseNormal=TextureSample(UVs=@n22) {samplerType=SAMPLERTYPE_Normal, texture=/Game/Athena/Apollo/Water/Textures/02_Rounded/02_WaterFoam_Veroni_base_N}.RGB
    AdditionalNormal=TextureSample(UVs=@n48) {samplerType=SAMPLERTYPE_Normal, texture=&02_WaterFoam_Veroni_base_N}.RGB
   ) {materialFunction=&BlendAngleCorrectedNormals}.Result
   Flatness=ScalarParameter() {defaultValue=0.9, parameterName=NormalFlatten}
  ) {materialFunction=&FlattenNormal}
  n57 = SetMaterialAttributes(
   Base Color=@n55
   Specular=Constant() {r=0.225}
   Roughness=ScalarParameter() {defaultValue=0.8, group=Added Foam, parameterName=Roughness2}
   Opacity=LinearInterpolate(
    A=ScalarParameter() {defaultValue=0.5, group=Added Foam, parameterName=Surface Opacity2_Edge}
    B=ScalarParameter() {defaultValue=1, group=Added Foam, parameterName=Surface Opacity2}
    Alpha=VertexColor().G
   )
   Normal=Normalize(VectorInput=Transform(None=@n56.Result))
  ) {attributeSetTypes [
    #1
    #2
    #3
    #4
    #5
    #6
   ]}
  n58 = GetMaterialAttributes(None=MaterialFunctionCall!(
    Base Material=MaterialFunctionCall!(Attributes=@n8, Depth) {materialFunction=/FortWater/Materials/Functions/TwoSidedWaterCorrections}.Result
    Top Material=@n57
    Alpha=MaterialFunctionCall(Transition Phase=@n9.R, Height Texture=@n30) {materialFunction=&HeightLerp}.Alpha
   ) {materialFunction=/Engine/Functions/MaterialLayerFunctions/MatLayerBlend_Standard}.Blended Material) {attributeGetTypes [
    #1
    #2
    #3
    #4
    #5
    #6
    #7
   ]}
  Add(A=@n58.World Position Offset, B=VectorParameter() {parameterName=HLOD Water Z Offset, defaultValue=<rgba 0 0 4 0>}.RGB)
  Constant3Vector() {constant=<rgba 15 0 0 0>}
  Constant3Vector() {constant=<rgba 0 15 0 0>}
  Constant() {r=0.225}
  Constant() {r=0.9}
  LinearInterpolate(A=@n55, B=Multiply(A=@n55, B=VectorParameter() {parameterName=FallsTint, defaultValue=<rgba 1 1 1 0>}.RGB))
  MaterialFunctionCall!(Flowmap Velocity, Previous Foam, Normal=@n7.Normal, WPO=@n7.Normal) {materialFunction=/FortWater/Materials/Functions/WaterSimulationTexture}
  Multiply(A=@n35, B=@n54)
  Reroute(None=@n58.Normal)
  Reroute(None=@n58.Opacity)
  Reroute(None=@n58.Roughness)
  Reroute(None=@n7.WPO)
  Reroute(None=@n58.Base Color)
  ScalarParameter() {defaultValue=0.3, group=Water Shading Model, parameterName=Anisotropy}
  n59 = Constant() {r=0.4}
  n60 = QualitySwitch(Default=@n58.Specular, Low=@n59, High=@n58.Specular, Medium=@n59)
  ShadingPathSwitch(Default=@n60, Deferred=@n60, Forward=@n59, Mobile=@n59)
  n61 = Constant() {r=1}
  n62 = QualitySwitch(Default=@n58.Refraction, Low=@n61, Medium=@n61)
  ShadingPathSwitch(Default=@n62, Deferred=@n62, Forward=@n61, Mobile=@n61)
  n63 = Saturate(None=TwoSidedSign())
  n64 = VectorParameter() {group=Water Shading Model, parameterName=Scattering, defaultValue=<rgba 0.1 0.1 0.1 0.5>}
  n65 = Divide(A=@n64.A) {constB=100}
  n66 = VectorParameter() {group=Water Shading Model, parameterName=Absorption, defaultValue=<rgba 1 8 20 100>}
  n67 = Multiply(A=@n66.RGB, B=@n66.RGB)
  SingleLayerWaterMaterialOutput!(
   ScatteringCoefficients=Multiply(A=@n63, B=Multiply(A=@n64.RGB, B=@n65))
   AbsorptionCoefficients=Multiply(A=@n63, B=Divide(B=@n67))
   PhaseG=CollectionParameter() {collection=/Water/Materials/MPC/MPC_Water, parameterName=Ansotropy}
   ColorScaleBehindWater
  )
 ]
}
