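# Two-sided Single Layer Water material for Apollo waterfalls, built on the FortWaterRiverFlowmaps function with velocity and depth blended between top and base values by a switchable UV mask.
# An added-foam region, set by a sphere mask, blends in color, specular, roughness, opacity and normal-direction vertex deformation.
# Its foam layers three panned swirl textures with noise-distorted, power-warped UVs.
# Parameters drive water scattering and absorption, and a switch offers a dithered vertical UV opacity mask.
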Material /Game/Effects/Fort_Effects/Materials/Environment/Apollo/M_Apollo_WaterFall_New
graph {
 properties {pixelDepthOffsetMode=PDOM_Legacy, shadingModel=MSM_SingleLayerWater, twoSided=true, usedShadingModels=SingleLayerWater}
 nodes [
  Constant3Vector() {constant=<rgba 1 0 0 0>}
  Constant() {r=150}
  n1 = SetMaterialAttributes!(
   MaterialAttributes
   Base Color=VectorParameter() {parameterName=BaseColor}.RGB
   Specular=ScalarParameter() {defaultValue=0.225, group=Water Shading Model, parameterName=Specular}
   Roughness=ScalarParameter() {defaultValue=0.1, group=Water Shading Model, parameterName=Roughness}
   Opacity=ScalarParameter() {group=Water Shading Model, parameterName=Surface Opacity}
   Normal=Constant3Vector() {constant=<rgba 0 0 1 0>}
   World Position Offset
  ) {attributeSetTypes [
    #1
    #2
    #3
    #4
    #5
    #6
   ]}
  n2 = VectorParameter() {parameterName=Water_Velocity, defaultValue=<rgba -800 0 600 256>}
  n3 = VectorParameter() {parameterName=Water_Velocity_Top, defaultValue=<rgba -800 0 600 256>}
  n4 = VectorParameter() {group=Added Foam, parameterName=JUSTVelocityFoamMask_Placement, defaultValue=<rgba 0 0.25 0 0>}
  n5 = TextureCoordinate()
  n6 = Sine(None=Add(A=ComponentMask(None=@n5) {g=true}, B=ScalarParameter() {defaultValue=-0.5, parameterName=JUSTVelocityFoamMask_Placement_02})) {period=2}
  n7 = StaticSwitchParameter(
   True=SphereMask(
    A=TextureCoordinate() {uTiling=0}
    B=ComponentMask(None=@n4.RGB) {g=true, r=true}
    Radius=ScalarParameter() {defaultValue=0.5, group=Added Foam, parameterName=JUSTVelocityFoamMask_Radius}
    Hardness=ScalarParameter() {defaultValue=0.5, group=Added Foam, parameterName=JUSTVelocityFoamMask_Hardness}
   )
   False=LinearInterpolate(A=Constant() {r=1}, B=Constant(), Alpha=Clamp(None=@n6))
  ) {defaultValue=true, parameterName=UseSphereMask?}
  n8 = MaterialFunctionCall!(
   Attributes=@n1
   Transition
   Sim Foam Erasure
   Velocity=LinearInterpolate(A=@n2.RGB, B=@n3.RGB, Alpha=@n7)
   Far Mask Override
   Depth=LinearInterpolate(A=@n2.A, B=@n3.A, Alpha=@n7)
  ) {materialFunction=/FortWater/Materials/Functions/FortWaterRiverFlowmaps}
  n9 = GetMaterialAttributes(None=@n8.Attributes) {attributeGetTypes [
    #1
    #2
    #3
    #4
    #5
    #6
   ]}
  n10 = SphereMask(
   A=TextureCoordinate() {uTiling=0}
   B=ComponentMask(None=VectorParameter() {group=Added Foam, parameterName=VelocityFoamMask_Placement, defaultValue=<rgba 0 0.25 0 0>}.RGB) {g=true, r=true}
   Radius=ScalarParameter() {defaultValue=0.5, group=Added Foam, parameterName=VelocityFoamMask_Radius}
   Hardness=ScalarParameter() {defaultValue=0.5, group=Added Foam, parameterName=VelocityFoamMask_Hardness}
  )
  LinearInterpolate(A=@n9.Specular, B=ScalarParameter() {group=Added Foam, parameterName=AddedFoam_Specular}, Alpha=@n10)
  n11 = TextureCoordinate()
  n12 = Power(Base=OneMinus(None=ComponentMask(None=@n11) {g=true}), Exp=ScalarParameter() {defaultValue=1, group=Added Foam, parameterName=FoamExp}) {constExponent=0.5}
  n13 = SphereMask(
   A=TextureCoordinate() {uTiling=0}
   B=ComponentMask(None=VectorParameter() {group=Added Foam, parameterName=AddedFoam_Mask_Placement, defaultValue=<rgba 0 0.25 0 0>}.RGB) {g=true, r=true}
   Radius=ScalarParameter() {defaultValue=0.5, group=Added Foam, parameterName=AddedFoam_Mask_Radius}
   Hardness=ScalarParameter() {group=Added Foam, parameterName=AddedFoam_Mask_Hardness}
  )
  n14 = LinearInterpolate(
   A=ScalarParameter() {defaultValue=-0.5, group=Added Foam, parameterName=Foam Texture Offset Top}
   B=ScalarParameter() {defaultValue=-0.5, group=Added Foam, parameterName=Foam Texture Offset}
   Alpha=@n13
  )
  n15 = TextureCoordinate() {vTiling=0.4}
  n16 = AppendVector(
   A=ComponentMask(None=@n15) {r=true}
   B=Power(Base=ComponentMask(None=@n15) {g=true}, Exp=ScalarParameter() {defaultValue=1.5, group=Added Foam, parameterName=EndTextureTile})
  )
  n17 = ComponentMask(None=VectorParameter() {group=Added Foam, parameterName=FoamTileMult, defaultValue=<rgba 1 5 0 0>}.RGB) {g=true, r=true}
  n18 = Time()
  n19 = Panner(
   Coordinate=Multiply(A=TextureCoordinate(), B=ScalarParameter() {defaultValue=0.7372, group=Added Foam, parameterName=UVDistortTilingA})
   Time=@n18
  ) {speedY=0.5}
  n20 = Panner(
   Coordinate=Multiply(A=TextureCoordinate() {uTiling=2}, B=ScalarParameter() {defaultValue=0.8, group=Added Foam, parameterName=UVDistortTilingB})
   Time=@n18
  ) {speedX=0.125, speedY=0.3}
  n21 = LinearInterpolate(Alpha=ComponentMask(None=TextureCoordinate()) {g=true}) {constB=1.5}
  n22 = LinearInterpolate(
   A=Multiply(
    A=TextureSample(UVs=@n19) {texture=/Game/Global/Textures/Default/Noise/T_Tiling_Noise_01}.G
    B=ScalarParameter() {group=Added Foam, parameterName=UVDistortAmountA}
   )
   B=Multiply(
    A=TextureSample(UVs=@n20) {texture=&T_Tiling_Noise_01}.G
    B=ScalarParameter() {group=Added Foam, parameterName=UVDistortAmountB}
   )
   Alpha=Clamp(None=OneMinus(None=@n21))
  )
  n23 = Add(A=OneMinus(None=Multiply(A=@n16, B=@n17)), B=@n22)
  n24 = ComponentMask(None=VectorParameter() {parameterName=FoamSpeed_01, defaultValue=<rgba 0 -1 0 0>}.RGB) {g=true, r=true}
  n25 = ComponentMask(None=TextureSample(UVs=Panner(Coordinate=@n23, Speed=@n24) {speedY=-1}) {samplerType=SAMPLERTYPE_LinearColor, texture=/Game/Global/Textures/Default/Caustics/03_WaterFoam_AccentSwirls_H}.G) {g=true}
  n26 = LinearInterpolate(
   A=ScalarParameter() {defaultValue=1, group=Added Foam, parameterName=FoamContrast Top}
   B=ScalarParameter() {defaultValue=1, group=Added Foam, parameterName=FoamContrast}
   Alpha=@n13
  )
  n27 = Multiply(A=@n12, B=LinearInterpolate(B=Add(A=@n14, B=@n25), Alpha=Clamp(None=@n26)) {constA=1})
  n28 = TextureCoordinate() {uTiling=2.8, vTiling=0.7}
  n29 = ComponentMask(None=@n28) {r=true}
  n30 = Power(Base=ComponentMask(None=@n28) {g=true}, Exp=ScalarParameter() {defaultValue=1.5, group=Added Foam, parameterName=EndTextureTile})
  n31 = OneMinus(None=Multiply(A=@n17, B=AppendVector(A=@n29, B=@n30)))
  n32 = VectorParameter() {parameterName=FoamSpeed_02, defaultValue=<rgba 0 -1 0 0>}
  n33 = TextureSample(UVs=Panner(Coordinate=Add(A=@n31, B=@n22), Speed=ComponentMask(None=@n32.RGB) {g=true, r=true}) {speedY=-1}) {samplerType=SAMPLERTYPE_LinearColor, texture=&03_WaterFoam_AccentSwirls_H}
  n34 = ScalarParameter() {defaultValue=1, group=Added Foam, parameterName=FoamContrast Top}
  n35 = ScalarParameter() {defaultValue=1, group=Added Foam, parameterName=FoamContrast}
  n36 = LinearInterpolate(B=Add(A=@n14, B=ComponentMask(None=@n33.G) {g=true}), Alpha=Clamp(None=LinearInterpolate(A=@n34, B=@n35, Alpha=@n13))) {constA=1}
  n37 = LinearInterpolate(A=@n27, B=Add(A=@n27, B=ComponentMask(None=@n36) {g=true}), Alpha=Clamp(None=@n12))
  n38 = TextureCoordinate() {uTiling=2, vTiling=0.6}
  n39 = ComponentMask(None=@n38) {g=true}
  n40 = ScalarParameter() {defaultValue=1.5, group=Added Foam, parameterName=EndTextureTile}
  n41 = Multiply(A=@n17, B=AppendVector(A=ComponentMask(None=@n38) {r=true}, B=Power(Base=@n39, Exp=@n40)))
  n42 = Panner(
   Coordinate=Add(A=OneMinus(None=@n41), B=@n22)
   Speed=ComponentMask(None=VectorParameter() {parameterName=FoamSpeed_03, defaultValue=<rgba 0 -1 0 0>}.RGB) {g=true, r=true}
  ) {speedY=-1}
  n43 = Add(
   A=@n14
   B=ComponentMask(None=TextureSample(UVs=@n42) {samplerType=SAMPLERTYPE_LinearColor, texture=&03_WaterFoam_AccentSwirls_H}.G) {g=true}
  )
  n44 = LinearInterpolate(
   A=ScalarParameter() {defaultValue=1, group=Added Foam, parameterName=FoamContrast Top}
   B=ScalarParameter() {defaultValue=1, group=Added Foam, parameterName=FoamContrast}
   Alpha=@n13
  )
  n45 = Power(Base=ComponentMask(None=TextureCoordinate()) {g=true}, Exp=ScalarParameter() {defaultValue=1, group=Added Foam, parameterName=FoamExp}) {constExponent=0.5}
  n46 = Multiply(
   A=LinearInterpolate(
    A=VectorParameter() {group=Added Foam, parameterName=WaterFallColor, defaultValue=<rgba 0.3185 0.5089 0.5029 1>}.RGB
    B=VectorParameter() {group=Added Foam, parameterName=WaterFallColor_Bottom, defaultValue=<rgba 0.3467 0.5647 0.5647 1>}.RGB
    Alpha=Clamp(None=@n37)
   )
   B=LinearInterpolate(A=@n37, B=LinearInterpolate(B=@n43, Alpha=@n44) {constA=1}, Alpha=Clamp(None=@n45)) {constAlpha=0.4}
  )
  LinearInterpolate(
   A=@n9.Base Color
   B=LinearInterpolate(
    A=VectorParameter() {group=Added Foam, parameterName=AddedFoam_BaseColor, defaultValue=<rgba 0.0006 0.117 0.2195 1>}.RGB
    Alpha=Clamp(None=@n46)
   )
   Alpha=@n10
  )
  n47 = ComponentMask(None=VectorParameter() {parameterName=AddedFoam_WPO_Tiling, defaultValue=<rgba 1 1 0 0>}.RGB) {g=true, r=true}
  n48 = Multiply(A=TextureCoordinate() {vTiling=2}, B=@n47)
  n49 = ComponentMask(None=VectorParameter() {group=Added Foam, parameterName=AddedFoam_WPO_Pan Speed, defaultValue=<rgba 0 -0.0065 0 1>}.RGB) {g=true, r=true}
  n50 = Multiply(A=@n47, B=TextureCoordinate() {uTiling=2, vTiling=4})
  n51 = Add(
   A=TextureSample(UVs=Panner(Coordinate=@n48, Speed=@n49) {bFractionalPart=true, speedX=0.025, speedY=-1}) {texture=&2ndInfectedMapLinear}.G
   B=TextureSample(UVs=Panner(Coordinate=@n50, Speed=@n49) {bFractionalPart=true, speedX=-0.05, speedY=-1}) {texture=&2ndInfectedMapLinear}.G
  )
  n52 = Multiply(
   A=Multiply(A=LinearInterpolate(Alpha=@n51) {constA=-1, constB=2}, B=VertexNormalWS())
   B=ScalarParameter() {defaultValue=1, group=Added Foam, parameterName=AddedFoam_Deform}
  )
  LinearInterpolate(A=@n52, B=VertexNormalWS())
  LinearInterpolate(A=@n9.World Position Offset, B=@n52, Alpha=@n10)
  MaterialFunctionCall!(Texture, Position) {materialFunction=/FortWater/Materials/Functions/WaterBodyData}
  n53 = ComponentMask(None=TextureCoordinate() {uTiling=0}) {g=true}
  n54 = ScalarParameter() {defaultValue=1}
  Multiply(A=MaterialFunctionCall(UV Channel=Add(A=@n53, B=@n54)) {materialFunction=&LinearGradient}.VGradient)
  Multiply(A=@n53, B=@n54)
  OneMinus(None=@n6)
  Reroute(None=LinearInterpolate(A=@n9.Opacity, B=ScalarParameter() {group=Added Foam, parameterName=AddedFoam_Opacity}, Alpha=@n10))
  Reroute(None=LinearInterpolate(A=@n9.Roughness, B=ScalarParameter() {group=Added Foam, parameterName=AddedFoam_Roughess}, Alpha=@n10))
  ScalarParameter() {group=Added Foam, parameterName=AddedFoam_WPO}
  n55 = VectorParameter() {group=Water Shading Model, parameterName=Scattering, defaultValue=<rgba 0.1 0.1 0.1 1.5>}
  n56 = Divide(A=@n55.A) {constB=100}
  n57 = VectorParameter() {group=Water Shading Model, parameterName=Absorption, defaultValue=<rgba 1 8 20 100>}
  SingleLayerWaterMaterialOutput!(
   ScatteringCoefficients=Add(A=@n8.Scattering, B=Multiply(A=@n55.RGB, B=@n56))
   AbsorptionCoefficients=Divide(B=Multiply(A=@n57.RGB, B=@n57.RGB))
   PhaseG=ScalarParameter() {defaultValue=0.5, group=Water Shading Model, parameterName=Anisotropy}
   ColorScaleBehindWater
  )
  n58 = ScalarParameter() {defaultValue=1, parameterName=OpacityMask_High}
  n59 = ScalarParameter() {parameterName=OpacityMask_Low}
  n60 = ComponentMask(None=TextureCoordinate()) {g=true}
  StaticSwitchParameter(
   True=MaterialFunctionCall(Alpha Threshold=LinearInterpolate(A=@n58, B=@n59, Alpha=@n60)) {materialFunction=&DitherTemporalAA}.Result
   False=Constant() {r=1}
  ) {parameterName=UseMaskUvSpace}
  TextureCoordinate()
  VertexColor()
 ]
}
